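# Niagara system for the SummitReedGrit emote that continuously spawns tumbling money meshes; only Money_02003 is enabled, the other five emitters are disabled variants.
# Grey or pink-red tinted meshes spawn on a cylinder shell, scale in and out, and fall under gravity, curl noise and drag.
# A low RegionMask gates point and vortex forces, swirling particles near the bottom alongside platform-gated bouncy collision.
# User.bIsFrontEndPreview retunes mask, collision size and vortex; time of day scales brightness.
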NiagaraSystem /BRCosmetics/Athena/Items/Cosmetics/Dances/BacklogEmotes/BR/SummitReedGrit/VFX/NS_Emote_SummitReedGrit_Money
dependencies {
 usedDataInterfaces [&NiagaraDataInterfaceArrayColor, &NiagaraDataInterfaceCurve, &NiagaraDataInterfacePlatformSet]
 usedDynamicInputs []
 usedModules [&SystemState]
}
emitters {
 Debug {
  bEnabled=false
  emitterUpdateScript [
   &EmitterState()
   &SpawnRate(SpawnRate=51, Use Spawn Probability=false) {module=SpawnRate001}
  ]
  particleSpawnScript [
   &InitializeParticle(
    Color=&SelectLinearColorFromArray(
     Array Sampling Mode=::NewEnumerator0
     Direct Array Index=?
     Color Selection Array=DataInterface() {propertyValues {ColorData [<rgba 0.9 0.81 0.225 1>, <rgba 1 0.45 0.5875 1>, <rgba 0.36 0.9 0.5535 1>, <rgba 0.8019 0.5 1 1>, <rgba 1 0.5127 0.3 1>]}}
     Random Seed=0
    )
    Lifetime Min=1.5
    Lifetime Mode=::NewEnumerator1
    Mesh Scale Mode=::NewEnumerator3
    Mesh Uniform Scale=0.85
    UsePositionOffset=false
    Write Material Random=true
   )
   &ShapeLocation(
    MaskSpawnGroup=false
    Non Uniform Scale=<1.2 1.2 1>
    Offset Mode=::NewEnumerator0
    Override Seed=false
    Shape Origin=$Particles.SkeletalMeshLocation.SampledPosition
    Sphere Radius=50
    UseOwnerScale=true
   )
   &InitialMeshOrientation(Use Rotation Vector=false)
   &ApplyInitialForces()
  ]
  particleUpdateScript [
   /SetVariables(Particles.Drag:NiagaraFloat=&FloatFromCurve(FloatCurve=DataInterface() {propertyValues {Curve {keys [{tangentMode=RCTM_User, time=0.1, value=0.5}, {tangentMode=RCTM_User, time=0.2, value=1}]}}})) {enabled=false, isSetParameters=true, module=SetVariables_#1}
   &ParticleState()
   &UpdateMeshOrientation(
    Rotation Rate=&RandomRangeFloat(Minimum=3, Maximum=5)
    Rotation Vector=&RandomVector(Vector Scale=1)
   )
   &PointForce(
    Force Falloff Distance=30
    Force Origin=&AddVectorToPosition(Position=/Niagara/DynamicInputs/Helpers/SimulationPosition(), Vector=<0 0 -25>)
    Force Strength=&Multiply_Float(A=&RandomRangeFloat(Minimum=1150, Maximum=1300), B=$Particles.RegionMask.Float)
    Linear Force Vector:Vector3f=&RandomRangeVector(Minimum=<-1 -1 0.2>, Maximum=<1 1 0.3>)
    Module.Use Falloff=false
    Module.UseLinearForce=true
   ) {enabled=false}
   &VortexForce(
    UseOriginOffset=true
    Vortex Force Amount=&Multiply_Float(A=&RandomRangeFloat(Minimum=-1000, Maximum=-1500), B=$Particles.RegionMask.Float)
    Vortex Origin Offset:Vector3f=<0 0 -50>
    Vortex Origin Offset Coordinate Space:ENiagaraCoordinateSpace=::Local
   ) {enabled=false}
   &ScaleMeshSize(Scale Factor=&VectorFromFloat(Value=&FloatFromCurve(
      FloatCurve=DataInterface() {propertyValues {Curve {keys [
          {arriveTangent=4.1785, interpMode=RCIM_Cubic, leaveTangent=4.5893, tangentMode=RCTM_Break, time=0, value=0}
          {arriveTangent=0.0224, interpMode=RCIM_Cubic, leaveTangent=0.0224, tangentMode=RCTM_Break, time=0.3, value=1}
         ]}}}
      CurveIndex=$Particles.Age
      Scale Curve=&FloatFromCurve(FloatCurve=DataInterface() {propertyValues {Curve {keys [
           {interpMode=RCIM_Cubic, tangentMode=RCTM_Break, time=0.7176, value=1}
           {arriveTangent=-9.1079, interpMode=RCIM_Cubic, tangentMode=RCTM_Break, time=1, value=0}
          ]}}})
     )))
   &RegionMask(
    Offset:Vector3f=<0 0 -200>
    Region Dimensions:Vector3f=<500 500 100>
    Region Falloff=0
    Region Type:ENiagaraDebugDrawShape=::NewEnumerator1
    UseOffset=true
    UseRotation=false
   )
   &VortexForce(
    UseOriginOffset=true
    Vortex Force Amount=&Multiply_Float(A=-9000, B=$Particles.RegionMask.Float)
    Vortex Origin Offset Coordinate Space:ENiagaraCoordinateSpace=::Local
   ) {module=VortexForce001}
   &GravityForce(Gravity=&RandomRangeVector(Minimum=<0 0 -1000>, Maximum=<0 0 -1200>))
   /Niagara/Modules/Update/Forces/CurlNoiseForce(Noise Frequency=5, Noise Strength=3000, Pan Noise=false) {module=CurlNoiseForce001}
   &Drag(Drag=3, Use Linear Drag=true, Use Rotational Drag=true)
   &AerodynamicDrag(
    Aerodynamic Drag:NiagaraFloat=$Particles.Drag
    Aerodynamic Rotational Drag:NiagaraFloat=$Particles.Drag
    Rotation Strength:NiagaraFloat=&RandomRangeFloat(Minimum=2, Maximum=2.4)
    Use Linear Drag=true
    Use Rotational Drag=true
   ) {enabled=false}
   &AerodynamicDrag(
    Aerodynamic Drag:NiagaraFloat=&RandomRangeFloat(Minimum=0.5, Maximum=0.6)
    Aerodynamic Rotational Drag:NiagaraFloat=&RandomRangeFloat(Minimum=0.1, Maximum=0.2)
    Drag Shape Curve=DataInterface() {propertyValues {Curve {keys [
        {interpMode=RCIM_Cubic, tangentMode=RCTM_User, time=0, value=0.1}
        {arriveTangent=2.0799, interpMode=RCIM_Cubic, leaveTangent=2.0799, tangentMode=RCTM_User, time=1, value=1}
       ]}}}
    Lift Shape Curve=DataInterface() {propertyValues {Curve {keys [
        {interpMode=RCIM_Cubic, tangentMode=RCTM_User, time=0, value=0}
        {interpMode=RCIM_Cubic, tangentMode=RCTM_User, time=0.5, value=1}
        {tangentMode=RCTM_User, time=1, value=0}
       ]}}}
    Rotation Strength:NiagaraFloat=&RandomRangeFloat(Minimum=20, Maximum=25)
    Use Linear Drag=true
    Use Rotational Drag=true
   ) {enabled=false, module=AerodynamicDrag002}
   &Collision(
    Age Colliding Particles=false
    CPU Friction Blending:ENiagaraFrictionMergeType=::NewEnumerator1
    Collision Enabled=&IsPlatformSetActive(New Platform Set=DataInterface() {propertyValues {Platforms {qualityLevelMask=-1}}})
    Friction=0
    KillOnCollision=false
    Module.Control Rest State Via Surface Normal=false
    Module.EnableMaxCollisionCount=false
    Module.Randomize Collision Normal Vector=false
    Radius Calculation Type=::NewEnumerator1
    Restitution=0
   ) {module=Collision002}
   &SolveForcesAndVelocity(Clamp Velocity=false, Limit Acceleration=false)
   &AlignSpriteToMeshOrientation(SetAlignment=true, SetFacing=true)
   &Color(Color=&LerpLinearColors(StartColor=<rgba 1 1 1 1>, EndColor=<rgba 1 0 0 1>, LerpFactor=$Particles.RegionMask.Float))
   &Color(Color=&MakeCustomLinearColorFromBool(Bool=$Particles.HasCollided, True Color=<rgba 1 0 0 1>, False Color=<rgba 1 1 1 1>)) {enabled=false, module=Color001}
  ]
  rendererClasses [&NiagaraMeshRendererProperties]
  renderers [{rendererClass=&NiagaraMeshRendererProperties, rendererIndex=0}]
 }
 Money {
  bEnabled=false
  emitterUpdateScript [
   &EmitterState()
   &SpawnRate(SpawnRate=15, Use Spawn Probability=false) {module=SpawnRate001}
  ]
  particleSpawnScript [
   &InitializeParticle(
    Color=&SelectLinearColorFromArray(
     Array Sampling Mode=::NewEnumerator0
     Direct Array Index=?
     Color Selection Array=DataInterface() {propertyValues {ColorData [<rgba 0.9 0.81 0.225 1>, <rgba 1 0.45 0.5875 1>, <rgba 0.36 0.9 0.5535 1>, <rgba 0.8019 0.5 1 1>, <rgba 1 0.5127 0.3 1>]}}
     Random Seed=0
    )
    Lifetime Max=1.5
    Lifetime Mode=::NewEnumerator1
    Mesh Scale Mode=::NewEnumerator3
    Mesh Uniform Scale=0.85
    UsePositionOffset=false
    Write Material Random=true
   )
   &ShapeLocation(
    MaskSpawnGroup=false
    Non Uniform Scale=<1.2 1.2 1>
    Offset Mode=::NewEnumerator0
    Override Seed=false
    Shape Origin=$Particles.SkeletalMeshLocation.SampledPosition
    Sphere Radius=20
    UseOwnerScale=true
   )
   &InitialMeshOrientation(Use Rotation Vector=false)
   &ApplyInitialForces()
  ]
  particleUpdateScript [
   /SetVariables(Particles.Drag:NiagaraFloat=&FloatFromCurve(FloatCurve=DataInterface() {propertyValues {Curve {keys [{tangentMode=RCTM_User, time=0.1, value=0.5}, {tangentMode=RCTM_User, time=0.2, value=1}]}}})) {enabled=false, isSetParameters=true, module=SetVariables_#1}
   &ParticleState()
   &RegionMask(Offset:Vector3f=<0 0 -50>, Region Falloff=0, Region Radius=40, UseOffset=true, UseRotation=false)
   &PointForce(
    Force Falloff Distance=30
    Force Origin=&AddVectorToPosition(Position=&SimulationPosition(), Vector=<0 0 -25>)
    Force Strength=&Multiply_Float(A=&RandomRangeFloat(Minimum=1150, Maximum=1300), B=$Particles.RegionMask.Float)
    Linear Force Vector:Vector3f=&RandomRangeVector(Minimum=<-1 -1 0.2>, Maximum=<1 1 0.3>)
    Module.Use Falloff=false
    Module.UseLinearForce=true
   )
   &ScaleMeshSize(Scale Factor=&VectorFromFloat(Value=&FloatFromCurve(
      FloatCurve=DataInterface() {propertyValues {Curve {keys [
          {arriveTangent=4.1785, interpMode=RCIM_Cubic, leaveTangent=4.5893, tangentMode=RCTM_Break, time=0, value=0}
          {arriveTangent=0.0224, interpMode=RCIM_Cubic, leaveTangent=0.0224, tangentMode=RCTM_Break, time=0.3, value=1}
         ]}}}
      CurveIndex=$Particles.Age
      Scale Curve=&FloatFromCurve(FloatCurve=DataInterface() {propertyValues {Curve {keys [
           {interpMode=RCIM_Cubic, tangentMode=RCTM_Break, time=0.7176, value=1}
           {arriveTangent=-9.1079, interpMode=RCIM_Cubic, tangentMode=RCTM_Break, time=1, value=0}
          ]}}})
     )))
   &GravityForce(Gravity=&RandomRangeVector(Minimum=<0 0 -1000>, Maximum=<0 0 -1100>))
   &CurlNoiseForce(Noise Frequency=4.9092, Noise Strength=1143.17, Pan Noise=false) {module=CurlNoiseForce001}
   &AerodynamicDrag(
    Aerodynamic Drag:NiagaraFloat=$Particles.Drag
    Aerodynamic Rotational Drag:NiagaraFloat=$Particles.Drag
    Rotation Strength:NiagaraFloat=&RandomRangeFloat(Minimum=2, Maximum=2.4)
    Use Linear Drag=true
    Use Rotational Drag=true
   ) {enabled=false}
   &AerodynamicDrag(
    Aerodynamic Drag:NiagaraFloat=&RandomRangeFloat(Minimum=0.5, Maximum=0.6)
    Aerodynamic Rotational Drag:NiagaraFloat=&RandomRangeFloat(Minimum=0.4, Maximum=0.8)
    Drag Shape Curve=DataInterface() {propertyValues {Curve {keys [
        {interpMode=RCIM_Cubic, tangentMode=RCTM_User, time=0, value=0.1}
        {arriveTangent=2.0799, interpMode=RCIM_Cubic, leaveTangent=2.0799, tangentMode=RCTM_User, time=1, value=1}
       ]}}}
    Lift Shape Curve=DataInterface() {propertyValues {Curve {keys [
        {interpMode=RCIM_Cubic, tangentMode=RCTM_User, time=0, value=0}
        {interpMode=RCIM_Cubic, tangentMode=RCTM_User, time=0.5, value=1}
        {tangentMode=RCTM_User, time=1, value=0}
       ]}}}
    Rotation Strength:NiagaraFloat=&RandomRangeFloat(Minimum=2, Maximum=2.4)
    Use Linear Drag=true
    Use Rotational Drag=true
   ) {module=AerodynamicDrag002}
   &Collision(
    Age Colliding Particles=false
    CPU Friction Blending:ENiagaraFrictionMergeType=::NewEnumerator1
    Collision Enabled=&IsPlatformSetActive(New Platform Set=DataInterface() {propertyValues {Platforms {qualityLevelMask=-8}}})
    Friction=0
    KillOnCollision=false
    Module.Control Rest State Via Surface Normal=false
    Module.EnableMaxCollisionCount=false
    Module.Randomize Collision Normal Vector=false
    Radius Calculation Type=::NewEnumerator1
    Restitution=0.8
   ) {module=Collision002}
   &SolveForcesAndVelocity(Clamp Velocity=false, Limit Acceleration=false)
   &AlignSpriteToMeshOrientation(SetAlignment=true, SetFacing=true)
   &Color(Color=&LerpLinearColors(StartColor=<rgba 1 1 1 1>, EndColor=<rgba 1 0 0 1>, LerpFactor=$Particles.RegionMask.Float)) {enabled=false}
  ]
  rendererClasses [&NiagaraMeshRendererProperties]
  renderers [{rendererClass=&NiagaraMeshRendererProperties, rendererIndex=0}]
 }
 Money_02 {
  bEnabled=false
  emitterUpdateScript [
   &EmitterState()
   &SpawnRate(SpawnRate=15, Use Spawn Probability=false) {module=SpawnRate001}
  ]
  particleSpawnScript [
   &InitializeParticle(
    Color=&SelectLinearColorFromArray(
     Array Sampling Mode=::NewEnumerator0
     Direct Array Index=?
     Color Selection Array=DataInterface() {propertyValues {ColorData [<rgba 0.9 0.81 0.225 1>, <rgba 1 0.45 0.5875 1>, <rgba 0.36 0.9 0.5535 1>, <rgba 0.8019 0.5 1 1>, <rgba 1 0.5127 0.3 1>]}}
     Random Seed=0
    )
    Lifetime Max=1.6
    Lifetime Mode=::NewEnumerator1
    Mesh Scale Mode=::NewEnumerator3
    Mesh Uniform Scale=0.85
    UsePositionOffset=false
    Write Material Random=true
   )
   &ShapeLocation(
    MaskSpawnGroup=false
    Non Uniform Scale=<1.2 1.2 1>
    Offset Mode=::NewEnumerator0
    Override Seed=false
    Shape Origin=$Particles.SkeletalMeshLocation.SampledPosition
    Sphere Radius=20
    UseOwnerScale=true
   )
   &InitialMeshOrientation(Use Rotation Vector=false)
   &ApplyInitialForces()
  ]
  particleUpdateScript [
   /SetVariables(Particles.Drag:NiagaraFloat=&FloatFromCurve(FloatCurve=DataInterface() {propertyValues {Curve {keys [{tangentMode=RCTM_User, time=0.1, value=0.5}, {tangentMode=RCTM_User, time=0.2, value=1}]}}})) {enabled=false, isSetParameters=true, module=SetVariables_#1}
   &ParticleState()
   &UpdateMeshOrientation(
    Rotation Rate=&RandomRangeFloat(Minimum=4, Maximum=5)
    Rotation Vector=&RandomVector(Vector Scale=1)
   )
   &RegionMask(
    Offset:Vector3f=<0 0 -52>
    Region Dimensions:Vector3f=<100 100 50>
    Region Falloff=0
    Region Type:ENiagaraDebugDrawShape=::NewEnumerator1
    UseOffset=true
    UseRotation=false
   )
   &Collision(
    Age Colliding Particles=false
    CPU Friction Blending:ENiagaraFrictionMergeType=::NewEnumerator1
    Collision Enabled=&IsPlatformSetActive(New Platform Set=DataInterface() {propertyValues {Platforms {qualityLevelMask=-1}}})
    Friction=0
    KillOnCollision=false
    Module.Control Rest State Via Surface Normal=false
    Module.EnableMaxCollisionCount=false
    Module.Randomize Collision Normal Vector=false
    Radius Calculation Type=::NewEnumerator1
    Restitution=&RandomRangeFloat(Minimum=0.7, Maximum=1)
   ) {module=Collision002}
   &PointForce(
    Force Falloff Distance=30
    Force Origin=&AddVectorToPosition(Position=&SimulationPosition(), Vector=<0 0 -25>)
    Force Strength=&Multiply_Float(A=&RandomRangeFloat(Minimum=2000, Maximum=2100), B=$Particles.RegionMask.Float)
    Linear Force Vector:Vector3f=&RandomRangeVector(Minimum=<-1 -1 0>, Maximum=<1 1 0>)
    Module.Use Falloff=false
    Module.UseLinearForce=true
   )
   &VortexForce(
    Influence Falloff Radius=300
    Use Influence Falloff=true
    UseOriginOffset=false
    Vortex Force Amount=&Multiply_Float(A=&RandomRangeFloat(Minimum=-6000, Maximum=-7000), B=$Particles.RegionMask.Float)
   )
   &PointForce(
    Force Falloff Distance=30
    Force Origin=&AddVectorToPosition(Position=&SimulationPosition(), Vector=<0 0 -25>)
    Force Strength=&MakeCustomFloatFromBool(Bool=$Particles.HasCollided, True Float=&RandomRangeFloat(Minimum=-1000, Maximum=-1200), False Float=0)
    Linear Force Vector:Vector3f=&RandomRangeVector(Minimum=<-1 -1 0>, Maximum=<1 1 0>)
    Module.Use Falloff=false
    Module.UseLinearForce=true
   ) {enabled=false, module=PointForce001}
   &VortexForce(
    Influence Falloff Radius=300
    Use Influence Falloff=true
    UseOriginOffset=false
    Vortex Force Amount=&MakeCustomFloatFromBool(Bool=$Particles.HasCollided, True Float=&RandomRangeFloat(Minimum=-5000, Maximum=-5600), False Float=0)
   ) {enabled=false, module=VortexForce001}
   &ScaleMeshSize(Scale Factor=&VectorFromFloat(Value=&FloatFromCurve(
      FloatCurve=DataInterface() {propertyValues {Curve {keys [
          {arriveTangent=4.1785, interpMode=RCIM_Cubic, leaveTangent=4.5893, tangentMode=RCTM_Break, time=0, value=0}
          {arriveTangent=0.0224, interpMode=RCIM_Cubic, leaveTangent=0.0224, tangentMode=RCTM_Break, time=0.3, value=1}
         ]}}}
      CurveIndex=$Particles.Age
      Scale Curve=&FloatFromCurve(FloatCurve=DataInterface() {propertyValues {Curve {keys [
           {interpMode=RCIM_Cubic, tangentMode=RCTM_Break, time=0.7176, value=1}
           {arriveTangent=-9.1079, interpMode=RCIM_Cubic, tangentMode=RCTM_Break, time=1, value=0}
          ]}}})
     )))
   &GravityForce(Gravity=&RandomRangeVector(Minimum=<0 0 -1000>, Maximum=<0 0 -1200>))
   &CurlNoiseForce(Noise Frequency=6, Noise Strength=1200, Pan Noise=false) {module=CurlNoiseForce001}
   &Drag(
    Drag=&RandomRangeFloat(Minimum=8, Maximum=10)
    Rotational Drag=2
    Use Linear Drag=true
    Use Rotational Drag=true
   )
   &AerodynamicDrag(
    Aerodynamic Drag:NiagaraFloat=$Particles.Drag
    Aerodynamic Rotational Drag:NiagaraFloat=$Particles.Drag
    Rotation Strength:NiagaraFloat=&RandomRangeFloat(Minimum=2, Maximum=2.4)
    Use Linear Drag=true
    Use Rotational Drag=true
   ) {enabled=false}
   &AerodynamicDrag(
    Aerodynamic Drag:NiagaraFloat=&RandomRangeFloat(Minimum=0.5, Maximum=0.6)
    Aerodynamic Rotational Drag:NiagaraFloat=&RandomRangeFloat(Minimum=0.1, Maximum=0.2)
    Drag Shape Curve=DataInterface() {propertyValues {Curve {keys [
        {interpMode=RCIM_Cubic, tangentMode=RCTM_User, time=0, value=0.1}
        {arriveTangent=2.0799, interpMode=RCIM_Cubic, leaveTangent=2.0799, tangentMode=RCTM_User, time=1, value=1}
       ]}}}
    Lift Shape Curve=DataInterface() {propertyValues {Curve {keys [
        {interpMode=RCIM_Cubic, tangentMode=RCTM_User, time=0, value=0}
        {interpMode=RCIM_Cubic, tangentMode=RCTM_User, time=0.5, value=1}
        {tangentMode=RCTM_User, time=1, value=0}
       ]}}}
    Rotation Strength:NiagaraFloat=&RandomRangeFloat(Minimum=20, Maximum=25)
    Use Linear Drag=true
    Use Rotational Drag=true
   ) {enabled=false, module=AerodynamicDrag002}
   &SolveForcesAndVelocity(Clamp Velocity=false, Limit Acceleration=false)
   &AlignSpriteToMeshOrientation(SetAlignment=true, SetFacing=true)
   &Color(Color=&LerpLinearColors(StartColor=<rgba 1 1 1 1>, EndColor=<rgba 1 0 0 1>, LerpFactor=$Particles.RegionMask.Float)) {enabled=false}
   &Color(Color=&MakeCustomLinearColorFromBool(Bool=$Particles.HasCollided, True Color=<rgba 1 0 0 1>, False Color=<rgba 1 1 1 1>)) {enabled=false, module=Color001}
  ]
  rendererClasses [&NiagaraMeshRendererProperties]
  renderers [{rendererClass=&NiagaraMeshRendererProperties, rendererIndex=0}]
 }
 Money_02001 {
  bEnabled=false
  emitterUpdateScript [
   &EmitterState()
   &SpawnRate(SpawnRate=15, Use Spawn Probability=false) {module=SpawnRate001}
  ]
  particleSpawnScript [
   &InitializeParticle(
    Color=&SelectLinearColorFromArray(
     Array Sampling Mode=::NewEnumerator0
     Direct Array Index=?
     Color Selection Array=DataInterface() {propertyValues {ColorData [<rgba 0.9 0.81 0.225 1>, <rgba 1 0.45 0.5875 1>, <rgba 0.36 0.9 0.5535 1>, <rgba 0.8019 0.5 1 1>, <rgba 1 0.5127 0.3 1>]}}
     Random Seed=0
    )
    Lifetime Max=1.6
    Lifetime Mode=::NewEnumerator1
    Mesh Scale Mode=::NewEnumerator3
    Mesh Uniform Scale=0.85
    UsePositionOffset=false
    Write Material Random=true
   )
   &ShapeLocation(
    Handle Radius=0
    Large Radius=25
    MaskSpawnGroup=false
    Offset Mode=::NewEnumerator0
    Override Seed=false
    Shape Origin=$Particles.SkeletalMeshLocation.SampledPosition
    Shape Primitive=::NewEnumerator3
    UseOwnerScale=true
   )
   &InitialMeshOrientation(Use Rotation Vector=false)
   &ApplyInitialForces()
  ]
  particleUpdateScript [
   &ParticleState()
   &UpdateMeshOrientation(
    Rotation Rate=&RandomRangeFloat(Minimum=2.7, Maximum=3)
    Rotation Vector=&RandomVector(Vector Scale=1)
   )
   &RegionMask(
    Offset:Vector3f=<0 0 -52>
    Region Dimensions:Vector3f=<100 100 50>
    Region Falloff=0
    Region Type:ENiagaraDebugDrawShape=::NewEnumerator1
    UseOffset=true
    UseRotation=false
   )
   &Collision(
    Age Colliding Particles=false
    CPU Friction Blending:ENiagaraFrictionMergeType=::NewEnumerator1
    Collision Enabled=&IsPlatformSetActive(New Platform Set=DataInterface() {propertyValues {Platforms {qualityLevelMask=-1}}})
    Friction=0
    KillOnCollision=false
    Module.Control Rest State Via Surface Normal=false
    Module.EnableMaxCollisionCount=false
    Module.Randomize Collision Normal Vector=false
    Radius Calculation Type=::NewEnumerator1
    Restitution=&RandomRangeFloat(Minimum=0.7, Maximum=1)
   ) {module=Collision002}
   &PointForce(
    Force Falloff Distance=30
    Force Origin=&AddVectorToPosition(Position=&SimulationPosition(), Vector=<0 0 -25>)
    Force Strength=&Multiply_Float(A=&RandomRangeFloat(Minimum=200, Maximum=300), B=$Particles.RegionMask.Float)
    Linear Force Vector:Vector3f=&RandomRangeVector(Minimum=<-1 -1 0>, Maximum=<1 1 0>)
    Module.Use Falloff=false
    Module.UseLinearForce=true
   )
   &VortexForce(
    Influence Falloff Radius=300
    Use Influence Falloff=true
    UseOriginOffset=false
    Vortex Force Amount=&Multiply_Float(A=&RandomRangeFloat(Minimum=-6000, Maximum=-7000), B=$Particles.RegionMask.Float)
   )
   &ScaleMeshSize(Scale Factor=&VectorFromFloat(Value=&FloatFromCurve(
      FloatCurve=DataInterface() {propertyValues {Curve {keys [
          {arriveTangent=4.1785, interpMode=RCIM_Cubic, leaveTangent=4.5893, tangentMode=RCTM_Break, time=0, value=0}
          {arriveTangent=0.0224, interpMode=RCIM_Cubic, leaveTangent=0.0224, tangentMode=RCTM_Break, time=0.3, value=1}
         ]}}}
      CurveIndex=$Particles.Age
      Scale Curve=&FloatFromCurve(FloatCurve=DataInterface() {propertyValues {Curve {keys [
           {interpMode=RCIM_Cubic, tangentMode=RCTM_Break, time=0.7176, value=1}
           {arriveTangent=-9.1079, interpMode=RCIM_Cubic, tangentMode=RCTM_Break, time=1, value=0}
          ]}}})
     )))
   &GravityForce(Gravity=&RandomRangeVector(Minimum=<0 0 -980>, Maximum=<0 0 -1100>))
   &CurlNoiseForce(Noise Frequency=6, Noise Strength=1200, Pan Noise=false) {module=CurlNoiseForce001}
   &Drag(
    Drag=&RandomRangeFloat(Minimum=8, Maximum=10)
    Rotational Drag=2
    Use Linear Drag=true
    Use Rotational Drag=true
   )
   &SolveForcesAndVelocity(Clamp Velocity=false, Limit Acceleration=false)
   &AlignSpriteToMeshOrientation(SetAlignment=true, SetFacing=true)
   &Color(Color=&LerpLinearColors(StartColor=<rgba 1 1 1 1>, EndColor=<rgba 1 0 0 1>, LerpFactor=$Particles.RegionMask.Float)) {enabled=false}
   &Color(Color=&MakeCustomLinearColorFromBool(Bool=$Particles.HasCollided, True Color=<rgba 1 0 0 1>, False Color=<rgba 1 1 1 1>)) {enabled=false, module=Color001}
  ]
  rendererClasses [&NiagaraMeshRendererProperties]
  renderers [{rendererClass=&NiagaraMeshRendererProperties, rendererIndex=0}]
 }
 Money_02002 {
  bEnabled=false
  emitterUpdateScript [
   &EmitterState()
   &SpawnRate(SpawnRate=15, Use Spawn Probability=false) {module=SpawnRate001}
  ]
  particleSpawnScript [
   &InitializeParticle(
    Color=&SelectLinearColorFromArray(
     Array Sampling Mode=::NewEnumerator0
     Direct Array Index=?
     Color Selection Array=DataInterface() {propertyValues {ColorData [<rgba 0.9 0.81 0.225 1>, <rgba 1 0.45 0.5875 1>, <rgba 0.36 0.9 0.5535 1>, <rgba 0.8019 0.5 1 1>, <rgba 1 0.5127 0.3 1>]}}
     Random Seed=0
    )
    Lifetime Max=1.6
    Lifetime Mode=::NewEnumerator1
    Mesh Scale Mode=::NewEnumerator3
    Mesh Uniform Scale=0.85
    UsePositionOffset=false
    Write Material Random=true
   )
   &ShapeLocation(
    Cylinder Height=30
    Cylinder Radius=25
    MaskSpawnGroup=false
    Offset Mode=::NewEnumerator0
    Override Seed=false
    Shape Origin=$Particles.SkeletalMeshLocation.SampledPosition
    Shape Primitive=::NewEnumerator1
    Surface Only=true
    Surface Only Band Thickness=0.4
    Use Endcaps In Surface Only Mode=false
    UseOwnerScale=true
   )
   &InitialMeshOrientation(Use Rotation Vector=false)
   &ApplyInitialForces()
  ]
  particleUpdateScript [
   &ParticleState()
   &UpdateMeshOrientation(
    Rotation Rate=&RandomRangeFloat(Minimum=1.7, Maximum=3)
    Rotation Vector=&RandomVector(Vector Scale=1)
   )
   &RegionMask(
    Offset:Vector3f=<0 0 -51.5>
    Region Dimensions:Vector3f=<130 130 50>
    Region Falloff=0
    Region Type:ENiagaraDebugDrawShape=::NewEnumerator1
    UseOffset=true
    UseRotation=false
   )
   &Collision(
    Age Colliding Particles=false
    CPU Friction Blending:ENiagaraFrictionMergeType=::NewEnumerator1
    Collision Enabled=&IsPlatformSetActive(New Platform Set=DataInterface() {propertyValues {Platforms {qualityLevelMask=-8}}})
    Friction=0
    KillOnCollision=false
    Module.Control Rest State Via Surface Normal=false
    Module.EnableMaxCollisionCount=false
    Module.Randomize Collision Normal Vector=false
    Radius Calculation Type=::NewEnumerator1
    Restitution=&RandomRangeFloat(Minimum=0.7, Maximum=1)
   ) {module=Collision002}
   &Collision(
    Age Colliding Particles=false
    CPU Collision Type=::NewEnumerator3
    CPU Friction Blending:ENiagaraFrictionMergeType=::NewEnumerator1
    Collision Enabled=&IsPlatformSetActive(New Platform Set=DataInterface() {propertyValues {Platforms {qualityLevelMask=-28}}})
    Friction=0
    KillOnCollision=false
    Module.Control Rest State Via Surface Normal=false
    Module.EnableMaxCollisionCount=false
    Module.Randomize Collision Normal Vector=false
    Radius Calculation Type=::NewEnumerator1
    Restitution=&RandomRangeFloat(Minimum=0.7, Maximum=1)
   )
   &PointForce(
    Force Falloff Distance=30
    Force Origin=&AddVectorToPosition(Position=&SimulationPosition(), Vector=<0 0 -25>)
    Force Strength=&Multiply_Float(A=&RandomRangeFloat(Minimum=350, Maximum=400), B=$Particles.RegionMask.Float)
    Linear Force Vector:Vector3f=&RandomRangeVector(Minimum=<-1 -1 0>, Maximum=<1 1 0>)
    Module.Use Falloff=false
    Module.UseLinearForce=true
   )
   &VortexForce(
    Influence Falloff Radius=300
    Use Influence Falloff=true
    UseOriginOffset=false
    Vortex Force Amount=&Multiply_Float(A=&RandomRangeFloat(Minimum=-5000, Maximum=-5500), B=$Particles.RegionMask.Float)
   )
   &ScaleMeshSize(Scale Factor=&VectorFromFloat(Value=&FloatFromCurve(
      FloatCurve=DataInterface() {propertyValues {Curve {keys [
          {arriveTangent=4.1785, interpMode=RCIM_Cubic, leaveTangent=4.5893, tangentMode=RCTM_Break, time=0, value=0}
          {arriveTangent=0.0224, interpMode=RCIM_Cubic, leaveTangent=0.0224, tangentMode=RCTM_Break, time=0.3, value=1}
         ]}}}
      CurveIndex=$Particles.Age
      Scale Curve=&FloatFromCurve(FloatCurve=DataInterface() {propertyValues {Curve {keys [
           {interpMode=RCIM_Cubic, tangentMode=RCTM_Break, time=0.7176, value=1}
           {arriveTangent=-9.1079, interpMode=RCIM_Cubic, tangentMode=RCTM_Break, time=1, value=0}
          ]}}})
     )))
   &GravityForce(Gravity=&RandomRangeVector(Minimum=<0 0 -980>, Maximum=<0 0 -1100>))
   &CurlNoiseForce(Noise Frequency=6, Noise Strength=1200, Pan Noise=false) {module=CurlNoiseForce001}
   &Drag(
    Drag=&RandomRangeFloat(Minimum=8, Maximum=10)
    Rotational Drag=2
    Use Linear Drag=true
    Use Rotational Drag=true
   )
   &SolveForcesAndVelocity(Clamp Velocity=false, Limit Acceleration=false)
   &AlignSpriteToMeshOrientation(SetAlignment=true, SetFacing=true)
   &Color(Color=&LerpLinearColors(StartColor=<rgba 1 1 1 1>, EndColor=<rgba 1 0 0 1>, LerpFactor=$Particles.RegionMask.Float)) {enabled=false}
   &Color(Color=&MakeCustomLinearColorFromBool(Bool=$Particles.HasCollided, True Color=<rgba 1 0 0 1>, False Color=<rgba 1 1 1 1>)) {enabled=false, module=Color001}
  ]
  rendererClasses [&NiagaraMeshRendererProperties]
  renderers [{rendererClass=&NiagaraMeshRendererProperties, rendererIndex=0}]
 }
 Money_02003 {
  emitterUpdateScript [
   &EmitterState()
   &SpawnRate(SpawnRate=15, Use Spawn Probability=false) {module=SpawnRate001}
  ]
  particleSpawnScript [
   &InitializeParticle(
    Color=&SelectLinearColorFromArray(
     Array Sampling Mode=::NewEnumerator0
     Direct Array Index=?
     Color Selection Array=DataInterface() {propertyValues {ColorData [<rgba 0.2031 0.2031 0.2031 1>, <rgba 1 0.2292 0.2292 1>]}}
     Random Seed=0
    )
    Lifetime Max=1.5
    Lifetime Min=1.1
    Lifetime Mode=::NewEnumerator1
    Mesh Scale Mode=::NewEnumerator3
    Mesh Uniform Scale=0.85
    UsePositionOffset=false
    Write Material Random=true
   )
   &ShapeLocation(
    Cylinder Height=30
    Cylinder Radius=25
    MaskSpawnGroup=false
    Offset Mode=::NewEnumerator0
    Override Seed=false
    Shape Origin=$Particles.SkeletalMeshLocation.SampledPosition
    Shape Primitive=::NewEnumerator1
    Surface Only=true
    Surface Only Band Thickness=0.4
    Use Endcaps In Surface Only Mode=false
    UseOwnerScale=true
   )
   &InitialMeshOrientation(Use Rotation Vector=false)
   &ApplyInitialForces()
  ]
  particleUpdateScript [
   &ParticleState()
   &UpdateMeshOrientation(
    Rotation Rate=&RandomRangeFloat(Minimum=1.7, Maximum=3)
    Rotation Vector=&RandomVector(Vector Scale=1)
   )
   &RegionMask(
    Offset:Vector3f=/Niagara/DynamicInputs/TypeConversions/MakeCustomPositionFromBool(Bool:NiagaraBool=$User.bIsFrontEndPreview, True Position:NiagaraPosition=<0 0 -66>, False Position:NiagaraPosition=<0 0 -75>)
    Region Dimensions:Vector3f=&MakeCustomVectorFromBool(Bool=$User.bIsFrontEndPreview, True Vector=<150 150 60>, False Vector=<150 150 55>)
    Region Falloff=0
    Region Type:ENiagaraDebugDrawShape=::NewEnumerator1
    UseOffset=true
    UseRotation=false
   )
   &Collision(
    Age Colliding Particles=false
    CPU Friction Blending:ENiagaraFrictionMergeType=::NewEnumerator1
    Collision Enabled=&IsPlatformSetActive(New Platform Set=DataInterface() {propertyValues {Platforms {qualityLevelMask=-8}}})
    Friction=0
    KillOnCollision=false
    Mesh Dimensions:Vector3f=&MakeCustomVectorFromBool(Bool=$User.bIsFrontEndPreview, True Vector=<4 4 4>, False Vector=<2 2 2>)
    Module.Control Rest State Via Surface Normal=false
    Module.EnableMaxCollisionCount=false
    Module.Randomize Collision Normal Vector=false
    Radius Calculation Type=::NewEnumerator1
    Restitution=&RandomRangeFloat(Minimum=0.7, Maximum=1)
   ) {module=Collision002}
   &Collision(
    Age Colliding Particles=false
    Analytical Collision Plane Position 1:Vector3f=<0 0 -85>
    Analytical Collision Plane Position 2:Vector3f=<0 0 -85>
    CPU Collision Type=::NewEnumerator3
    CPU Friction Blending:ENiagaraFrictionMergeType=::NewEnumerator1
    Collision Enabled=&IsPlatformSetActive(New Platform Set=DataInterface() {propertyValues {Platforms {qualityLevelMask=-28}}})
    Friction=0
    KillOnCollision=false
    Module.Control Rest State Via Surface Normal=false
    Module.EnableMaxCollisionCount=false
    Module.Randomize Collision Normal Vector=false
    Radius Calculation Type=::NewEnumerator1
    Restitution=&RandomRangeFloat(Minimum=0.7, Maximum=1)
   ) {enabled=false}
   &PointForce(
    Force Falloff Distance=30
    Force Origin=&AddVectorToPosition(Position=&SimulationPosition(), Vector=<0 0 -55>)
    Force Strength=&Multiply_Float(A=&RandomRangeFloat(Minimum=500, Maximum=600), B=$Particles.RegionMask.Float)
    Linear Force Vector:Vector3f=&RandomRangeVector(Minimum=<-1 -1 0.1>, Maximum=<1 1 0.15>)
    Module.Use Falloff=false
    Module.UseLinearForce=true
   )
   &PointForce(
    Force Falloff Distance=30
    Force Origin=&AddVectorToPosition(Position=&SimulationPosition(), Vector=<0 0 -55>)
    Force Strength=&Multiply_Float(A=&RandomRangeFloat(Minimum=2000, Maximum=2000), B=$Particles.RegionMask.Float)
    Linear Force Vector:Vector3f=&RandomRangeVector(Minimum=<-1 -1 0.1>, Maximum=<1 1 0.15>)
    Module.Use Falloff=false
    Module.UseLinearForce=true
   ) {enabled=false, module=PointForce001}
   &VortexForce(
    Influence Falloff Radius=300
    Use Influence Falloff=true
    UseOriginOffset=false
    Vortex Force Amount=&Multiply_Float(
     A=&MakeCustomFloatFromBool(
      Bool=$User.bIsFrontEndPreview
      True Float=&RandomRangeFloat(Minimum=-6700, Maximum=-6900)
      False Float=&RandomRangeFloat(Minimum=-5600, Maximum=-6000)
     )
     B=$Particles.RegionMask.Float
    )
   )
   &ScaleMeshSize(Scale Factor=&VectorFromFloat(Value=&FloatFromCurve(
      FloatCurve=DataInterface() {propertyValues {Curve {keys [
          {arriveTangent=4.1785, interpMode=RCIM_Cubic, leaveTangent=4.5893, tangentMode=RCTM_Break, time=0, value=0}
          {arriveTangent=0.0224, interpMode=RCIM_Cubic, leaveTangent=0.0224, tangentMode=RCTM_Break, time=0.3, value=1}
         ]}}}
      CurveIndex=$Particles.Age
      Scale Curve=&FloatFromCurve(FloatCurve=DataInterface() {propertyValues {Curve {keys [
           {interpMode=RCIM_Cubic, tangentMode=RCTM_Break, time=0.7176, value=1}
           {arriveTangent=-9.1079, interpMode=RCIM_Cubic, tangentMode=RCTM_Break, time=1, value=0}
          ]}}})
     )))
   &GravityForce(Gravity=&RandomRangeVector(Minimum=<0 0 -980>, Maximum=<0 0 -1100>))
   &CurlNoiseForce(Noise Frequency=6, Noise Strength=1200, Pan Noise=false) {module=CurlNoiseForce001}
   &Drag(
    Drag=&RandomRangeFloat(Minimum=8, Maximum=10)
    Rotational Drag=2
    Use Linear Drag=true
    Use Rotational Drag=true
   )
   &SolveForcesAndVelocity(Clamp Velocity=false, Limit Acceleration=false)
   &AlignSpriteToMeshOrientation(SetAlignment=true, SetFacing=true)
   &Color(Color=&LerpLinearColors(StartColor=<rgba 1 1 1 1>, EndColor=<rgba 1 0 0 1>, LerpFactor=$Particles.RegionMask.Float)) {enabled=false}
   &Color(Color=&MakeCustomLinearColorFromBool(Bool=$Particles.HasCollided, True Color=<rgba 1 0 0 1>, False Color=<rgba 1 1 1 1>)) {enabled=false, module=Color001}
   &ScaleColor(
    Scale RGB=&VectorFromFloat(Value=&FNTimeOfDayScalar(Night Scale=1.6, Morning Scale=1, Evening Scale=1))
    ScaleA=true
    ScaleRGB=true
   )
  ]
  rendererClasses [&NiagaraMeshRendererProperties]
  renderers [{rendererClass=&NiagaraMeshRendererProperties, rendererIndex=0}]
 }
}
summary {
 systemName=NS_Emote_SummitReedGrit_Money
 emitters [
  {bEnabled=false, emitterName=Money, rendererClasses [&NiagaraMeshRendererProperties]}
  {bEnabled=false, emitterName=Money_02, rendererClasses [&NiagaraMeshRendererProperties]}
  {bEnabled=false, emitterName=Debug, rendererClasses [&NiagaraMeshRendererProperties]}
  {bEnabled=false, emitterName=Money_02001, rendererClasses [&NiagaraMeshRendererProperties]}
  {bEnabled=false, emitterName=Money_02002, rendererClasses [&NiagaraMeshRendererProperties]}
  {emitterName=Money_02003, rendererClasses [&NiagaraMeshRendererProperties]}
 ]
 userVariables [
  User.bIsFrontEndPreview {type=NiagaraBool}
  User.blsFrontEnd {type=NiagaraBool}
 ]
}
